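# Niagara swing trail for the off-hand of the CrispSeason drumstick pickaxe cosmetic, built from two emitters.
# FN_AnimTrail draws short-lived anim-trail ribbons through six renderers, cycling colour with system age and fading over ribbon life.
# Particles_Swing spawns teal sprites by rate and movement distance from a socket reader, inheriting slight swing velocity before drag and gravity act.
# Both emitters tint by a system time-of-day colour and drive dynamic material parameters.
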NiagaraSystem /SparksCosmetics/Instrument/DrumStick/CrispSeason/Effects/NS_Pickaxe_DrumStick_CrispSeason_OffHandTrail
dependencies {
 usedDataInterfaces [
  &NiagaraDataInterfaceCurve
  &NiagaraDataInterfaceColorCurve
  &NiagaraDataInterfaceSpriteRendererInfo
  &NiagaraDataInterfaceVector2DCurve
  &NiagaraDataInterfaceSocketReader
 ]
 usedDynamicInputs []
 usedModules [
  &self:SystemSpawnScript.SystemScriptSource.SystemScriptGraph.NiagaraNodeAssignment_0.SetVariables_#1
  &self:SystemSpawnScript.SystemScriptSource.SystemScriptGraph.NiagaraNodeAssignment_1.SetVariables_#2
  &self:SystemSpawnScript.SystemScriptSource.SystemScriptGraph.NiagaraNodeAssignment_2.SetVariables_#3
  &SystemState
 ]
}
emitters {
 FN_AnimTrail {
  emitterSpawnScript [
   &FN_RibbonIDOffsetUtility()
   /SetVariables(Emitter.RibbonLife=0.17) {isSetParameters=true}
  ]
  emitterUpdateScript [
   &EmitterState()
   &FN_SpawnAnimTrail(AnimTrailWidthMode=::FromFirst)
   &FN_SpawnAnimTrail_v2(AnimTrailWidthMode=::FromSecond) {enabled=false}
  ]
  particleSpawnScript [
   &FN_InitializeAnimTrailBoneIndecies(First Bone or Socket Index=1, Second Bone or Socket Index=0)
   &FN_InitializeAnimTrailBoneIndecies(AnimTrail Selection=1, First Bone or Socket Index=2, Second Bone or Socket Index=3) {enabled=false, module=FN_InitializeAnimTrailBoneIndecies001}
   &FN_InitializeAnimTrail(AnimTrailWidth=0.45, Lifetime=$Emitter.RibbonLife)
   &AddVelocity(Velocity=<0 400 0>) {enabled=false}
   /SetVariables(Particles.Spawn Alpha:NiagaraFloat=&FloatFromCurve(
     FloatCurve=DataInterface() {propertyValues {Curve {defaultValue=1, keys [<at 0.05 0 RCIM_Cubic>,<at 0.1 1 RCIM_Cubic>,<at 0.2 0 RCIM_Cubic>]}}}
     CurveIndex=$Emitter.Age
    )) {isSetParameters=true}
  ]
  particleUpdateScript [
   &SolveForcesAndVelocity(Clamp Velocity=false, Limit Acceleration=false)
   &FN_UpdateAnimTrail() {module=FN_UpdateAnimTrail001}
   &ScaleColor(
    Linear Color Curve=DataInterface() {propertyValues {
      AlphaCurve {keys [<at 0.0009 0>,<at 0.0167 1>,<at 0.7193 1>]}
      BlueCurve {keys [<at 0 1>,<at 1 1>]}
      GreenCurve {keys [<at 0 1>,<at 1 1>]}
      RedCurve {keys [<at 0 1>,<at 1 1>]}
     }}
    Scale Mode=2
   ) {enabled=false}
   &ScaleColor(
    Curve Index=&Divide_Float(A=$Emitter.Age, B=$Emitter.RibbonLife)
    Linear Color Curve=DataInterface() {propertyValues {
      AlphaCurve {keys [<at 0.5615 1>,<at 1 0>]}
      BlueCurve {keys [<at 0 1>,<at 1 1>]}
      GreenCurve {keys [<at 0 1>,<at 1 1>]}
      RedCurve {keys [<at 0 1>,<at 1 1>]}
     }}
    Scale Mode=2
   ) {enabled=false, module=ScaleColor001}
   &ScaleColor(
    Scale Alpha=&Multiply_Float(
     A=&FloatFromCurve(
      FloatCurve=DataInterface() {propertyValues {Curve {keys [<at 0.3 1>,<at 0.6091 0.6294>,<at 1 0>]}}}
      CurveIndex=&Divide_Float(A=$Emitter.Age, B=$Emitter.RibbonLife)
     )
     B=$Particles.Initial.Spawn Alpha
    )
    Scale RGB=&Multiply_Vector(
     A=&MakeVectorFromLinearColorRGB(Color=$System.TOD Color)
     B=&MakeVectorFromLinearColorRGB(Color=&ColorFromCurve(
       ColorCurve=DataInterface() {propertyValues {
         AlphaCurve {keys [<at 0 1 RCIM_Cubic>]}
         BlueCurve {keys [<at 0.2647 0.6308 RCIM_Cubic>,<at 1 0.6068 RCIM_Cubic>]}
         GreenCurve {keys [<at 0.2647 0.6867 RCIM_Cubic>,<at 1 0.2713 RCIM_Cubic>]}
         RedCurve {keys [<at 0.2647 0 RCIM_Cubic>,<at 1 0 RCIM_Cubic>]}
        }}
       CurveIndex=&Frac_Float(Float=&Multiply_Float(A=$Engine.System.Age, B=7))
      ))
    )
    ScaleA=true
    ScaleRGB=true
   ) {module=ScaleColor005}
   &ScaleColor(
    Scale Mode=::NewEnumerator1
    Scale RGBA=&MakeVector4_FromLinearColor(LinearColor=$System.TOD Color)
    ScaleRGBA=true
   ) {module=ScaleColor004}
   &ScaleColor(
    Scale Alpha=&FloatFromCurve(
     FloatCurve=DataInterface() {propertyValues {Curve {keys [<at 0.4 1>,<at 0.56 0.9>,<at 1 0>]}}}
     CurveIndex=&Divide_Float(A=$Emitter.Age, B=$Emitter.RibbonLife)
    )
    ScaleA=true
    ScaleRGB=true
   ) {enabled=false, module=ScaleColor003}
   &DynamicMaterialParameters(
    Index 0 Param 1=&FloatFromCurve(
     FloatCurve=DataInterface() {propertyValues {Curve {keys [<at 0.5415 1 RCIM_Cubic>, {arriveTangent=-5.2255, interpMode=RCIM_Cubic, leaveTangent=-5.2255, tangentMode=RCTM_User, time=1, value=0}]}}}
     CurveIndex=&Divide_Float(A=$Emitter.Age, B=$Emitter.RibbonLife)
     Scale Curve=0.85
    )
    Index 0 Param 2=1
    Module.Index0Float0Write=true
    Module.Index0Float1Write=true
    Module.Index0Float2Write=true
    Module.Index0Float3Write=true
   )
   &DynamicMaterialParameters(
    Index 0 Param 1=&FloatFromCurve(
     FloatCurve=DataInterface() {propertyValues {Curve {keys [<at 0.3676 1>,<at 0.7294 0.6529>,<at 1 0>]}}}
     CurveIndex=&Divide_Float(A=$Emitter.Age, B=$Emitter.RibbonLife)
     Scale Curve=0.9
    )
    Module.Index0Float0Write=true
    Module.Index0Float1Write=true
    Module.Index0Float2Write=true
    Module.Index0Float3Write=true
   ) {enabled=false, module=DynamicMaterialParameters001}
   &ScaleColor(
    Curve Index=&Divide_Float(A=$Emitter.Age, B=0.4)
    Linear Color Curve=DataInterface() {propertyValues {
      AlphaCurve {keys [<at 0 1 RCIM_Constant>]}
      BlueCurve {keys [<at 0 1 RCIM_Constant>,<at 0.25 0.0179 RCIM_Constant>,<at 0.5 1 RCIM_Constant>,<at 0.75 0 RCIM_Constant>]}
      GreenCurve {keys [<at 0 0.1186 RCIM_Constant>,<at 0.25 1 RCIM_Constant>,<at 0.5 0.08 RCIM_Constant>,<at 0.75 0.048 RCIM_Constant>]}
      RedCurve {keys [<at 0 0.7578 RCIM_Constant>,<at 0.25 0 RCIM_Constant>,<at 0.5 0 RCIM_Constant>,<at 0.75 1 RCIM_Constant>]}
     }}
    Scale Mode=2
   ) {enabled=false, module=ScaleColor002}
   &ScaleRibbonWidth(Ribbon Width Scale=&FloatFromCurve(
     FloatCurve=DataInterface() {propertyValues {Curve {keys [
         {arriveTangent=6.4267, interpMode=RCIM_Cubic, leaveTangent=6.4267, tangentMode=RCTM_User, time=0, value=0.5206}
         {arriveTangentWeight=0.0409, interpMode=RCIM_Cubic, leaveTangentWeight=0.2419, tangentMode=RCTM_User, time=0.1013, value=0.9912}
         {arriveTangentWeight=0.2971, interpMode=RCIM_Cubic, leaveTangentWeight=0.1489, tangentMode=RCTM_User, time=0.7, value=1}
         {arriveTangent=-4.73, interpMode=RCIM_Cubic, leaveTangent=-4.73, tangentMode=RCTM_User, time=1, value=0}
        ]}}}
     CurveIndex=&Divide_Float(A=$Emitter.Age, B=$Emitter.RibbonLife)
    )) {enabled=false}
   /SetVariables(Particles.RibbonUVDistance:NiagaraFloat=&Multiply_FloatByInt(Float=1.4, Integer=&ReturnParticleIDAttributes())) {isSetParameters=true}
   /SetVariables(Particles.ScreenRibbonWidth:NiagaraFloat=&Multiply_Float(A=$Particles.RibbonWidth, B=1)) {isSetParameters=true}
  ]
  rendererClasses [&NiagaraRibbonRendererProperties]
  renderers [
   {rendererClass=&NiagaraRibbonRendererProperties}
   {rendererClass=&NiagaraRibbonRendererProperties, rendererIndex=1}
   {rendererClass=&NiagaraRibbonRendererProperties, rendererIndex=2}
   {rendererClass=&NiagaraRibbonRendererProperties, rendererIndex=3}
   {rendererClass=&NiagaraRibbonRendererProperties, rendererIndex=4}
   {rendererClass=&NiagaraRibbonRendererProperties, rendererIndex=5}
  ]
 }
 Particles_Swing {
  emitterUpdateScript [
   &EmitterState(Life Cycle Mode=::NewEnumerator1, Loop Behavior=::NewEnumerator1, Loop Delay=0.01, Loop Duration=1, UseLoopDelay=true)
   &SpawnRate(Spawn Probability=0.9, SpawnRate=7, Use Spawn Probability=false)
   &SpawnPerUnit(Module.Use Max Movement Threshold=true, Module.Use Movement Tolerance=false, Spawn Spacing=40, Use Spawn Probability=false)
  ]
  particleSpawnScript [
   &InitializeParticle(
    Color=<rgba 0 0.6867 0.6308 1>
    Lifetime Max=0.54
    Lifetime Min=0.3
    Lifetime Mode=::NewEnumerator1
    Sprite Rotation Mode=::NewEnumerator2
    Sprite Size Max=<3 8>
    Sprite Size Min=<1 4>
    Sprite Size Mode=::NewEnumerator2
    UsePositionOffset=false
    Write Material Random=true
   )
   &SocketLocation(
    Interpolate Socket Sampling=true
    SamplePosOffset=false
    Socket Index=1
    Socket Reader=$System.Socket Reader
    Socket Sampling Mode=::NewEnumerator3
    VelocitySpeedLimit=false
   )
   &ShapeLocation(MaskSpawnGroup=false, Override Seed=false, Sphere Radius=12, UseOwnerScale=true)
   &InheritVelocity(Inherited Velocity Amount Scale=<0.05 0.05 0.05>, Speed Limit=true, UseSpeedThreshold=false)
   &AddVelocity(Rotation Coordinate Space=::World, Velocity=$Output.SocketLocation.SampledVelocity, Velocity Speed Scale=0.05)
   &ApplyInitialForces()
   &SubUVAnimation(
    Evaluation Type=::NewEnumerator0
    "Sprite Renderer "=DataInterface() {propertyValues {SpriteRenderer=&self:Particles_Swing_0.NiagaraSpriteRendererProperties_1}}
    SubUV Animation Mode:ENiagara_SubUVLookupModeV2=::NewEnumerator2
    UseEndFrame=false
    UseStartFrame=false
   )
   /SetVariables(Particles.VisibilityTag=&RandomRangeInt(Minimum=0, Maximum=1)) {isSetParameters=true}
  ]
  particleUpdateScript [
   &ParticleState()
   &ScaleSpriteSize(
    Non-Uniform Curve Scale=&RandomRangeVector2D(Minimum=<1 0.6>, Maximum=<1 1>)
    Non-Uniform Curve Sprite Scale=DataInterface() {propertyValues {
      XCurve {keys [{tangentMode=RCTM_User, time=0, value=1}]}
      YCurve {keys [{arriveTangent=2.6251, interpMode=RCIM_Cubic, leaveTangent=2.6251, tangentMode=RCTM_User, time=0, value=0.5}, <at 0.5 1 RCIM_Cubic>]}
     }}
    Scale Sprite Size Mode=::NewEnumerator3
   )
   &Drag(Drag=&RandomRangeFloat(Minimum=2, Maximum=4), Use Linear Drag=true, Use Rotational Drag=false)
   &GravityForce(Gravity=&RandomRangeVector(Minimum=<0 0 -120>, Maximum=<0 0 -216>))
   &Drag(Drag=&RandomRangeFloat(Minimum=1, Maximum=7), Use Linear Drag=true, Use Rotational Drag=false) {enabled=false, module=Drag001}
   &InheritVelocity(
    Inherited Velocity Amount Scale=&Multiply_VectorByFloat(
     Vector=<1 1 1>
     Float=&FloatFromCurve(
      FloatCurve=DataInterface() {propertyValues {Curve {
         defaultValue=1
         keys [
          {arriveTangent=-2.0203, interpMode=RCIM_Cubic, leaveTangent=-2.0203, tangentMode=RCTM_User, time=0, value=1}
          {interpMode=RCIM_Cubic, tangentMode=RCTM_User, time=1, value=0}
         ]
        }}}
      Scale Curve=0.1
     )
    )
    Inherited Velocity Speed Limit=20
    Speed Limit=false
    UseSpeedThreshold=false
   ) {enabled=false, module=InheritVelocity001}
   &SolveForcesAndVelocity(Clamp Velocity=false, Limit Acceleration=false)
   &DynamicMaterialParameters(
    Index 0 Param 1=&RandomRangeFloat(Minimum=0, Maximum=1)
    Index 0 Param 2=&FloatFromCurve(FloatCurve=DataInterface() {propertyValues {Curve {keys [<at 0.35 0.45>,<at 1 0.9>]}}})
    Index 0 Param 4=&FloatFromCurve(FloatCurve=DataInterface() {propertyValues {Curve {keys [<at 0 0 RCIM_Cubic>, {arriveTangent=2.332, interpMode=RCIM_Cubic, leaveTangent=2.332, tangentMode=RCTM_User, time=1, value=1}]}}})
    Module.Index0Float0Write=true
    Module.Index0Float1Write=true
    Module.Index0Float2Write=false
    Module.Index0Float3Write=true
   )
   &ScaleColor(Scale RGB=&MakeVectorFromLinearColorRGB(Color=$System.TOD Color), ScaleA=true, ScaleRGB=true)
  ]
  rendererClasses [&NiagaraSpriteRendererProperties]
  renderers [
   {rendererClass=&NiagaraSpriteRendererProperties}
   {rendererClass=&NiagaraSpriteRendererProperties, rendererIndex=1}
  ]
 }
}
summary {
 systemName=NS_Pickaxe_DrumStick_CrispSeason_OffHandTrail
 emitters [
  {emitterName=FN_AnimTrail, rendererClasses [&NiagaraRibbonRendererProperties]}
  {emitterName=Particles_Swing, rendererClasses [&NiagaraSpriteRendererProperties]}
 ]
 userVariables [
  User.bIsFrontEnd {type=NiagaraBool}
  User.bIsFrontEndPreview {type=NiagaraBool}
  User.IsFrontEnd {type=NiagaraBool}
  User.RibbonIDOffset {type=NiagaraInt32}
  User.SkeletalMesh {
   type=NiagaraDataInterfaceSkeletalMesh
   defaultValue {
    struct=NiagaraExt_VariableValue_DataInterface
    value {dataInterface=&self:NiagaraDataInterfaceSkeletalMesh_0}
   }
  }
 ]
}
