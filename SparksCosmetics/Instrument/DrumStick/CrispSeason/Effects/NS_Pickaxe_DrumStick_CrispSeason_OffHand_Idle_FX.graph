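# Niagara idle effect for the off-hand CrispSeason drumstick pickaxe, with hanging drips, falling droplets and swing particles.
# Drips are pendulum-constrained meshes oriented toward the emitter origin, followed by a delayed burst of gravity-driven sprites.
# Swing sprites spawn at a socket by rate and movement, gated by User.IsPickaxe, then fade and shrink with distance.
# System speed values drive drip visibility and swing spawning, time-of-day color tints everything, and two debug emitters are disabled.
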NiagaraSystem /SparksCosmetics/Instrument/DrumStick/CrispSeason/Effects/NS_Pickaxe_DrumStick_CrispSeason_OffHand_Idle_FX
dependencies {
 usedDataInterfaces [
  &NiagaraDataInterfaceVectorCurve
  &NiagaraDataInterfaceParticleRead
  &NiagaraDataInterfaceCurve
  &NiagaraDataInterfaceVector2DCurve
  &NiagaraDataInterfaceSpriteRendererInfo
  &NiagaraDataInterfaceSkeletalMesh
  &NiagaraDataInterfaceSocketReader
  &NiagaraDataInterfaceColorCurve
 ]
 usedDynamicInputs [&ColorFromCurve]
 usedModules [
  &self:SystemSpawnScript.SystemScriptSource.SystemScriptGraph.NiagaraNodeAssignment_1.SetVariables_#1
  &self:SystemSpawnScript.SystemScriptSource.SystemScriptGraph.NiagaraNodeAssignment_4.SetVariables_#2
  &self:SystemSpawnScript.SystemScriptSource.SystemScriptGraph.NiagaraNodeAssignment_7.SetVariables_#3
  &self:SystemSpawnScript.SystemScriptSource.SystemScriptGraph.NiagaraNodeAssignment_3.SetVariables_#4
  &SystemState
  &self:SystemSpawnScript.SystemScriptSource.SystemScriptGraph.NiagaraNodeAssignment_5.SetVariables_#5
  &self:SystemSpawnScript.SystemScriptSource.SystemScriptGraph.NiagaraNodeAssignment_2.SetVariables_#6
  &self:SystemSpawnScript.SystemScriptSource.SystemScriptGraph.NiagaraNodeAssignment_6.SetVariables_#7
  &self:SystemSpawnScript.SystemScriptSource.SystemScriptGraph.NiagaraNodeAssignment_0.SetVariables_#8
 ]
}
emitters {
 DripSprites {
  emitterUpdateScript [
   &EmitterState()
   &SpawnBurst_Instantaneous(Spawn Time=0.35, Use Spawn Probability=false, UseLoopCountLimit=false)
   &SpawnRate(Spawn Probability=0.9, SpawnRate=0.6, Use Spawn Probability=false) {enabled=false}
   &SpawnParticlesFromOtherEmitter(
    Attribute Reader=DataInterface() {propertyValues {EmitterBinding {bindingMode=Other, emitterName=Drips}}}
    Minimum Age Spawn Threshold=0.3
    Spawn Rate=1
    Use Maximum Age=false
    Use Minimum Age=false
   ) {enabled=false}
  ]
  particleSpawnScript [
   &InitializeParticle(
    Color=$System.TOD Color
    Lifetime Max=0.6
    Lifetime Min=0.3
    Lifetime Mode=::NewEnumerator1
    Position Offset=<0 0 -0.5>
    Position Offset Coordinate Space=::World
    Sprite Size Mode=::NewEnumerator1
    Uniform Sprite Size Max=2.5
    Uniform Sprite Size Min=2
    UsePositionOffset=true
    Write Material Random=true
   )
   &SampleParticlesFromOtherEmitter(Position Sampling:ENiagara_DirectReadSamplingMode=::NewEnumerator3) {enabled=false}
   /SetVariables(Particles.Color=&MultiplyLinearColor(Linear Color=$Particles.Color, Scale Factor=&MakeVector4(X=1, Y=1, Z=1, W=$System.SystemSpeedAlpha))) {isSetParameters=true}
   &SocketLocation(
    Interpolate Socket Sampling=true
    SamplePosOffset=false
    Socket Reader=$System.Socket Reader
    Socket Sampling Mode=::NewEnumerator3
    VelocitySpeedLimit=false
   ) {enabled=false}
   &ShapeLocation(MaskSpawnGroup=false, Override Seed=false, Sphere Radius=12, UseOwnerScale=true) {enabled=false}
   &InheritVelocity(Inherited Velocity Amount Scale=<0.25 0.25 0.25>, Inherited Velocity Speed Limit=20, Speed Limit=false, UseSpeedThreshold=false)
   &CameraOffset(Camera Offset Amount=5) {enabled=false}
  ]
  particleUpdateScript [
   &ParticleState()
   &Drag(Drag=&RandomRangeFloat(Minimum=1, Maximum=7), Use Linear Drag=true, Use Rotational Drag=false)
   &GravityForce(Gravity=<0 0 -500>)
   &ScaleColor(
    Scale Alpha=&Multiply_Float(
     A=&FloatFromCurve(FloatCurve=DataInterface() {propertyValues {Curve {
         defaultValue=0
         keys [{interpMode=RCIM_Cubic, tangentMode=RCTM_User, time=0.4403, value=1}, {interpMode=RCIM_Cubic, tangentMode=RCTM_User, time=1, value=0}]
        }}})
     B=$System.SystemSpeedAlpha
    )
    ScaleA=true
    ScaleRGB=true
   )
   &ScaleSpriteSize(
    Non-Uniform Curve Sprite Scale=DataInterface() {propertyValues {
      XCurve {keys [
        {arriveTangent=2.0238, interpMode=RCIM_Cubic, leaveTangent=2.0238, tangentMode=RCTM_User, time=0, value=0.8}
        {arriveTangent=-0.6083, interpMode=RCIM_Cubic, leaveTangent=-0.6083, tangentMode=RCTM_User, time=0.5, value=1}
        <at 1 0.7 RCIM_Cubic>
       ]}
      YCurve {keys [
        {arriveTangent=3.6421, interpMode=RCIM_Cubic, leaveTangent=3.6421, tangentMode=RCTM_User, time=0, value=0.3}
        {arriveTangent=1.3, interpMode=RCIM_Cubic, leaveTangent=1.3, time=0.5, value=1}
        {arriveTangent=1.2034, interpMode=RCIM_Cubic, leaveTangent=1.2034, tangentMode=RCTM_User, time=1, value=1.6}
       ]}
     }}
    Scale Sprite Size Mode=::NewEnumerator3
   ) {enabled=false}
   &ScaleSpriteSize(
    Uniform Curve Scale=$System.EngineOwnerScale
    Uniform Curve Sprite Scale=DataInterface() {propertyValues {Curve {keys [
        {arriveTangent=2.753, interpMode=RCIM_Cubic, leaveTangent=2.753, tangentMode=RCTM_User, time=0, value=0.3494}
        {arriveTangent=0.143, interpMode=RCIM_Cubic, leaveTangent=0.143, tangentMode=RCTM_User, time=0.3012, value=0.9559}
        {interpMode=RCIM_Cubic, tangentMode=RCTM_User, time=1, value=0.1882}
       ]}}}
   ) {module=ScaleSpriteSize001}
   &SolveForcesAndVelocity(Clamp Velocity=false, Limit Acceleration=false)
  ]
  rendererClasses [&NiagaraSpriteRendererProperties]
  renderers [{rendererClass=&NiagaraSpriteRendererProperties}]
 }
 Drips {
  emitterUpdateScript [
   &EmitterState()
   &SpawnBurst_Instantaneous(Use Spawn Probability=false, UseLoopCountLimit=false)
   &SpawnRate(
    Spawn Probability=0.9
    SpawnRate=&RandomRangeFloat(Minimum=0, Maximum=2, Evaluation Type=::NewEnumerator1)
    Use Spawn Probability=false
   ) {enabled=false}
  ]
  particleSpawnScript [
   &InitializeParticle(
    Lifetime=0.4
    Mesh Scale=<0.1 0.1 0.05>
    Mesh Scale Mode=::NewEnumerator4
    Sprite Rotation Mode=::NewEnumerator2
    UsePositionOffset=false
    Write Material Random=true
   )
   &AddVelocity(Velocity=<0 0 -1>) {enabled=false}
   &InitialMeshOrientation(
    Mesh Axis To Orient=<0 0 1>
    Mesh Orientation Mode=::NewEnumerator2
    Orientation Coordinate Space=::World
    Orientation Vector=<0 0 1>
    Use Rotation Vector=false
   )
  ]
  particleUpdateScript [
   &ParticleState()
   /SetVariables(Particles.MeshOrigin:NiagaraPosition=$Engine.Emitter.SimulationPosition) {isSetParameters=true}
   &ScaleColor(
    Scale Mode=::NewEnumerator1
    Scale RGBA=&MakeVector4_FromLinearColor(LinearColor=$System.TOD Color)
    ScaleRGBA=true
   ) {module=ScaleColor002}
   &Drag(Drag=5, Use Linear Drag=true, Use Rotational Drag=false)
   &GravityForce(Gravity=<0 0 -900>)
   &PendulumSetup(
    Module.Enable Spring Driver=false
    Pendulum Length=10
    Pendulum Pivot:Vector3f=$Particles.MeshOrigin
    "Spring Driven Constraint (Tightness)"=3
   )
   &SolveForcesAndVelocity(Clamp Velocity=false, Limit Acceleration=false)
   &PendulumConstraint()
   &CalculateAccurateVelocity()
   /SetVariables(Particles.MeshFacing:Vector3f=&NormalizeVector(
     Vector To Normalize=&Subtract_Vector(A=$Particles.MeshOrigin, B=Linked() {linkedVariable {name=Particles.Position, type=NiagaraPosition}})
     Safe Normalize=true
     Safe Fallback Threshold=0
     Safe Fallback Vector=<0 0 0.001>
    )) {isSetParameters=true}
   &ScaleMeshSize(Scale Factor=&VectorFromCurve(
     VectorCurve=DataInterface() {propertyValues {
       XCurve {keys [<at 0 0>,<at 0.1 1>]}
       YCurve {keys [<at 0 0>,<at 0.1 1>]}
       ZCurve {keys [
         {arriveTangent=3.0998, interpMode=RCIM_Cubic, leaveTangent=3.0998, tangentMode=RCTM_User, time=0, value=0}
         {arriveTangent=1.2, interpMode=RCIM_Cubic, leaveTangent=1.2, time=0.225, value=0.4856}
         {arriveTangent=0.1106, interpMode=RCIM_Cubic, leaveTangent=-2.4267, tangentMode=RCTM_Break, time=0.75, value=0.9}
         {arriveTangent=-0.4385, interpMode=RCIM_Cubic, leaveTangent=-0.4385, tangentMode=RCTM_User, time=1, value=0.6}
        ]}
      }}
     Scale Curve=&VectorFromFloat(Value=$System.EngineOwnerScale)
    ))
   &ScaleMeshSize(Scale Factor=&VectorFromFloat(Value=$System.SystemSpeedAlpha)) {module=ScaleMeshSize001}
   /Niagara/Modules/Update/Orientation/V2/OrientMeshToVector(
    Look At Direction:Vector3f=&Multiply_VectorByFloat(Vector=$Particles.MeshFacing, Float=1)
    Override Quat=false
    Reference Axis Vector:Vector3f=<0 0 1>
    UseOptionalSideAxis=false
    UseOptionalUpAxis=false
   )
   &DynamicMaterialParameters(
    Index 2 Param 1=HlslExpression() {hlslExpression=Particles.MeshOrigin.x}
    Index 2 Param 2=HlslExpression() {hlslExpression=Particles.MeshOrigin.y}
    Index 2 Param 3=HlslExpression() {hlslExpression=Particles.MeshOrigin.z}
    Index 2 Param 4=1.5
    Index 3 Param 1=&MakeFloatFromVector(Vector=Linked() {linkedVariable {name=Particles.Position, type=NiagaraPosition}}, Channel=::NewEnumerator0)
    Index 3 Param 2=&MakeFloatFromVector(Vector=Linked() {linkedVariable {name=Particles.Position, type=NiagaraPosition}}, Channel=::NewEnumerator1)
    Index 3 Param 3=&MakeFloatFromVector(Vector=Linked() {linkedVariable {name=Particles.Position, type=NiagaraPosition}}, Channel=::NewEnumerator2)
    Index 3 Param 4=0.2
    Module.Param0WriteEnabled2=true
    Module.Param0WriteEnabled3=true
    Module.Param1WriteEnabled2=true
    Module.Param1WriteEnabled3=true
    Module.Param2WriteEnabled2=true
    Module.Param2WriteEnabled3=true
    Module.Param3WriteEnabled2=true
    Module.Param3WriteEnabled3=true
    Write Parameter Index 0=false
    Write Parameter Index 2=true
    Write Parameter Index 3=true
   )
  ]
  rendererClasses [&NiagaraMeshRendererProperties]
  renderers [{rendererClass=&NiagaraMeshRendererProperties, rendererIndex=0}]
 }
 Particles_Swing {
  emitterUpdateScript [
   &EmitterState()
   &SpawnRate(
    Spawn Probability=&MakeCustomFloatFromBool(Bool=&SetBoolByFloatComparison(A=$User.IsPickaxe, B=0), True Float=0, False Float=1)
    SpawnRate=&Multiply_Float(A=&RandomRangeFloat(Minimum=7, Maximum=10, Evaluation Type=::NewEnumerator1), B=$System.SpeedSpawnRate)
    Use Spawn Probability=true
   )
   &SpawnPerUnit(
    Module.Use Max Movement Threshold=true
    Module.Use Movement Tolerance=false
    Spawn Probability:NiagaraFloat=&MakeCustomFloatFromBool(Bool=&SetBoolByFloatComparison(A=$User.IsPickaxe, B=0), True Float=0, False Float=$System.SpeedSpawnRate)
    Spawn Spacing=300
    Use Spawn Probability=true
   )
  ]
  particleSpawnScript [
   &InitializeParticle(
    Color=<rgba 0 0.5778 0.5307 0.7>
    Lifetime Max=0.3
    Lifetime Min=0.25
    Lifetime Mode=::NewEnumerator1
    Sprite Rotation Mode=::NewEnumerator2
    Sprite Size Max=<2 8>
    Sprite Size Min=<1 4>
    Sprite Size Mode=::NewEnumerator2
    UsePositionOffset=false
    Write Material Random=true
   )
   &SocketLocation(
    Interpolate Socket Sampling=true
    SamplePosOffset=false
    Socket Index=1
    Socket Reader=$System.Socket Reader
    Socket Sampling Mode=::NewEnumerator3
    VelocitySpeedLimit=false
   )
   &ShapeLocation(MaskSpawnGroup=false, Override Seed=false, Sphere Radius=7, UseOwnerScale=true)
   &InheritVelocity(Inherited Velocity Amount Scale=<0.05 0.05 0.05>, Speed Limit=true, UseSpeedThreshold=false)
   &AddVelocity(Rotation Coordinate Space=::World, Velocity=$Output.SocketLocation.SampledVelocity, Velocity Speed Scale=-0.2)
   &AddVelocity(
    Module.Use Falloff Distance=false
    Origin Offset=<-10 0 0>
    Override Seed=false
    Velocity Mode=::NewEnumerator1
    Velocity Speed=&RandomRangeFloat(Minimum=0, Maximum=25)
   ) {module=AddVelocity001}
   &ApplyInitialForces()
   &SubUVAnimation(
    Evaluation Type=::NewEnumerator0
    "Sprite Renderer "=DataInterface() {propertyValues {SpriteRenderer=&self:NiagaraEmitter_2.NiagaraSpriteRendererProperties_1}}
    SubUV Animation Mode:ENiagara_SubUVLookupModeV2=::NewEnumerator2
    UseEndFrame=false
    UseStartFrame=false
   )
   /SetVariables(Particles.VisibilityTag=&RandomRangeInt(Minimum=0, Maximum=1)) {isSetParameters=true}
  ]
  particleUpdateScript [
   &ParticleState()
   &Collision(
    Age Colliding Particles=false
    Enable Rest State=false
    KillOnCollision=true
    Module.EnableMaxCollisionCount=false
    Module.Randomize Collision Normal Vector=false
   )
   /SetVariables(
    Particles.DistanceScale:NiagaraFloat=/Game/Effects/Niagara/DynamicInputs/Math/FN_Fit(
     Clamp:NiagaraBool=true
     Value:NiagaraFloat=&FloatFromCurve(
      FloatCurve=DataInterface() {propertyValues {Curve {keys [{interpMode=RCIM_Cubic, leaveTangentWeight=0.5003, tangentMode=RCTM_User, time=0.6, value=1}, <at 0.7 0 RCIM_Cubic>]}}}
      CurveIndex=&Divide_Float(
       A=&VectorLength(Vector=&SubtractVectorFromPosition(Position=$Particles.Position, Vector=Linked() {linkedVariable {name=Engine.Emitter.SimulationPosition, type=NiagaraPosition}}))
       B=100
      )
     )
     SourceMin:NiagaraFloat=0
     SourceMax:NiagaraFloat=1
     DestMin:NiagaraFloat=0
     DestMax:NiagaraFloat=1
    )
    Particles.DistanceScaleOpacity:NiagaraFloat=&FN_Fit(
     Clamp:NiagaraBool=true
     Value:NiagaraFloat=&FloatFromCurve(
      FloatCurve=DataInterface() {propertyValues {Curve {keys [{arriveTangent=-7.5653, interpMode=RCIM_Cubic, leaveTangent=-7.5653, tangentMode=RCTM_User, time=0.5, value=1}, <at 0.75 0 RCIM_Cubic>]}}}
      CurveIndex=&Divide_Float(
       A=&VectorLength(Vector=&SubtractVectorFromPosition(Position=$Particles.Position, Vector=Linked() {linkedVariable {name=Engine.Emitter.SimulationPosition, type=NiagaraPosition}}))
       B=100
      )
     )
     SourceMin:NiagaraFloat=0
     SourceMax:NiagaraFloat=1
     DestMin:NiagaraFloat=0
     DestMax:NiagaraFloat=1
    )
   ) {isSetParameters=true}
   &ScaleSpriteSize(
    Non-Uniform Curve Scale=&Multiply_Vector2D_ByFloat(Float=$Particles.DistanceScale, Vector2D=&RandomRangeVector2D(Minimum=<1 0.6>, Maximum=<1 1>))
    Non-Uniform Curve Sprite Scale=DataInterface() {propertyValues {
      XCurve {keys [
        {tangentMode=RCTM_User, time=0, value=1}
        {arriveTangentWeight=0.2848, interpMode=RCIM_Cubic, leaveTangentWeight=0.1222, tangentMode=RCTM_User, time=0.7, value=1}
        {
         arriveTangent=-4.5348
         arriveTangentWeight=0.1002
         interpMode=RCIM_Cubic
         leaveTangent=-4.5348
         tangentMode=RCTM_User
         time=1.0005
         value=0.3
        }
       ]}
      YCurve {keys [
        {arriveTangent=2.6251, interpMode=RCIM_Cubic, leaveTangent=2.6251, tangentMode=RCTM_User, time=0, value=0.5}
        {arriveTangent=0.6941, interpMode=RCIM_Cubic, leaveTangent=0.6941, time=0.5, value=1}
        {arriveTangent=-1.4, interpMode=RCIM_Cubic, leaveTangent=-1.4, time=0.7, value=0.9859}
        {arriveTangent=-4.8274, interpMode=RCIM_Cubic, leaveTangent=-4.8274, tangentMode=RCTM_User, time=1, value=0.3}
       ]}
     }}
    Scale Sprite Size Mode=::NewEnumerator3
   )
   &Drag(Drag=&RandomRangeFloat(Minimum=3, Maximum=4), Use Linear Drag=true, Use Rotational Drag=false)
   &GravityForce(Gravity=&RandomRangeVector(Minimum=<0 0 -75>, Maximum=<0 0 -150>))
   &Drag(Drag=&RandomRangeFloat(Minimum=1, Maximum=7), Use Linear Drag=true, Use Rotational Drag=false) {enabled=false, module=Drag001}
   &InheritVelocity(
    Inherited Velocity Amount Scale=&Multiply_VectorByFloat(
     Vector=<1 1 1>
     Float=&FloatFromCurve(
      FloatCurve=DataInterface() {propertyValues {Curve {
         defaultValue=1
         keys [
          {arriveTangent=-2.0203, interpMode=RCIM_Cubic, leaveTangent=-2.0203, tangentMode=RCTM_User, time=0, value=1}
          {interpMode=RCIM_Cubic, tangentMode=RCTM_User, time=1, value=0}
         ]
        }}}
      Scale Curve=0.1
     )
    )
    Inherited Velocity Speed Limit=20
    Speed Limit=false
    UseSpeedThreshold=false
   ) {enabled=false, module=InheritVelocity001}
   &SolveForcesAndVelocity(Clamp Velocity=false, Limit Acceleration=false)
   &DynamicMaterialParameters(
    Index 0 Param 1=&RandomRangeFloat(Minimum=0, Maximum=1)
    Index 0 Param 2=&FloatFromCurve(FloatCurve=DataInterface() {propertyValues {Curve {keys [<at 0.35 0.45>,<at 1 0.9>]}}})
    Index 0 Param 4=&FloatFromCurve(FloatCurve=DataInterface() {propertyValues {Curve {keys [<at 0 0 RCIM_Cubic>, {arriveTangent=2.332, interpMode=RCIM_Cubic, leaveTangent=2.332, tangentMode=RCTM_User, time=1, value=1}]}}})
    Module.Index0Float0Write=true
    Module.Index0Float1Write=true
    Module.Index0Float2Write=false
    Module.Index0Float3Write=true
   )
   &ScaleColor(
    Scale Alpha=$Particles.DistanceScaleOpacity
    Scale RGB=&MakeVectorFromLinearColorRGB(Color=$System.TOD Color)
    ScaleA=true
    ScaleRGB=true
   )
  ]
  rendererClasses [&NiagaraSpriteRendererProperties]
  renderers [
   {rendererClass=&NiagaraSpriteRendererProperties}
   {rendererClass=&NiagaraSpriteRendererProperties, rendererIndex=1}
  ]
 }
 debugdistance {
  bEnabled=false
  emitterUpdateScript [
   &EmitterState()
   &SpawnRate(SpawnRate=0.25, Use Spawn Probability=false)
   &SpawnBurst_Instantaneous(Use Spawn Probability=false, UseLoopCountLimit=true) {enabled=false}
  ]
  particleSpawnScript [
   &InitializeParticle(
    Color Mode=::NewEnumerator0
    Lifetime=5
    Sprite Size Mode=::NewEnumerator3
    Uniform Sprite Size=100
    UsePositionOffset=false
    Write Material Random=true
   )
   &SkeletalMeshLocation(MaskSpawnGroup=false, Mesh Sampling Type=::NewEnumerator5, Skeletal Mesh=$System.Skeletal Mesh, Socket Sampling Mode=::NewEnumerator5)
  ]
  particleUpdateScript [
   &ParticleState()
   &DynamicMaterialParameters(
    Index 0 Param 1=&VectorLength(Vector=&SubtractVectorFromPosition(Position=$Particles.Position, Vector=Linked() {linkedVariable {name=Engine.Emitter.SimulationPosition, type=NiagaraPosition}}))
    Module.Index0Float0Write=true
    Module.Index0Float1Write=true
    Module.Index0Float2Write=true
    Module.Index0Float3Write=true
   )
  ]
  rendererClasses [&NiagaraSpriteRendererProperties, &NiagaraMeshRendererProperties]
  renderers [
   {rendererClass=&NiagaraSpriteRendererProperties}
   {rendererClass=&NiagaraMeshRendererProperties, rendererIndex=1}
  ]
 }
 debugspeed {
  bEnabled=false
  emitterUpdateScript [
   &EmitterState()
   &SpawnBurst_Instantaneous(Use Spawn Probability=false, UseLoopCountLimit=true)
  ]
  particleSpawnScript [&InitializeParticle(
    Color Mode=::NewEnumerator0
    Lifetime=5
    Sprite Size Mode=::NewEnumerator3
    Uniform Sprite Size=100
    UsePositionOffset=false
    Write Material Random=true
   )]
  particleUpdateScript [
   &ParticleState(Kill Particles When Lifetime Has Elapsed=false)
   &SkeletalMeshLocation(MaskSpawnGroup=false, Mesh Sampling Type=::NewEnumerator5, Skeletal Mesh=$System.Skeletal Mesh, Socket Sampling Mode=::NewEnumerator5)
   &DynamicMaterialParameters(
    Index 0 Param 1=&VectorLength(Vector=$Engine.Owner.Velocity)
    Module.Index0Float0Write=true
    Module.Index0Float1Write=true
    Module.Index0Float2Write=true
    Module.Index0Float3Write=true
   )
  ]
  rendererClasses [&NiagaraSpriteRendererProperties, &NiagaraMeshRendererProperties]
  renderers [
   {rendererClass=&NiagaraSpriteRendererProperties}
   {rendererClass=&NiagaraMeshRendererProperties, rendererIndex=1}
  ]
 }
}
summary {
 systemName=NS_Pickaxe_DrumStick_CrispSeason_OffHand_Idle_FX
 emitters [
  {emitterName=Drips, rendererClasses [&NiagaraMeshRendererProperties]}
  {emitterName=DripSprites, rendererClasses [&NiagaraSpriteRendererProperties]}
  {emitterName=Particles_Swing, rendererClasses [&NiagaraSpriteRendererProperties]}
  {
   bEnabled=false
   emitterName=debugspeed
   rendererClasses [&NiagaraSpriteRendererProperties, &NiagaraMeshRendererProperties]
  }
  {
   bEnabled=false
   emitterName=debugdistance
   rendererClasses [&NiagaraSpriteRendererProperties, &NiagaraMeshRendererProperties]
  }
 ]
 userVariables [
  User.bIsFrontEnd {type=NiagaraBool}
  User.bIsFrontEndPreview {type=NiagaraBool}
  User.IsFrontEnd {type=NiagaraBool}
  User.IsPickaxe {type=NiagaraFloat, defaultValue=0, description=""}
 ]
}
